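# Niagara skydiving contrail for the SnowKnight cosmetic, attaching crystal and glow effects to two sampled points on the character's skeletal mesh.
# Persistent crystal shards and lights follow those points along cylinders that progressively bend with owner velocity.
# Mesh chunks, glow sprites, sparks and tapering ribbons add the trail, with some velocities varying for front end or dive.
# User Crystal_Color tints most emitters, with an IsOrangeRed flag correcting hue and saturation.
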NiagaraSystem /BRCosmetics/Athena/Effects/Particles/Skydiving/NS_SnowKnight/NS_Contrail_SknowKnight
dependencies {
 usedDataInterfaces [
  &NiagaraDataInterfaceCurve
  &NiagaraDataInterfaceArrayFloat
  &NiagaraDataInterfaceColorCurve
  &NiagaraDataInterfaceMeshRendererInfo
 ]
 usedDynamicInputs [
  &NormalizeVector
  &FN_FrontEndFloatModifier
  &SetBoolByFloatComparison
  &Divide_Float
  &self:IsColor
  &MakeCustomLinearColorFromBool
 ]
 usedModules [
  /Game/Effects/Niagara/Modules/Skydiving/FN_DivingAttributes/v5/FN_DivingAttributes
  &self:SystemSpawnScript.SystemScriptSource.SystemScriptGraph.NiagaraNodeAssignment_0.SetVariables_#1
  &SystemState
  &Timeline
  &self:ScratchModule_02
  &self:SystemSpawnScript.SystemScriptSource.SystemScriptGraph.NiagaraNodeAssignment_1.SetVariables_#2
  &self:SystemSpawnScript.SystemScriptSource.SystemScriptGraph.NiagaraNodeAssignment_2.SetVariables_#3
 ]
}
emitters {
 ChunksHand {
  emitterUpdateScript [
   &EmitterState(Life Cycle Mode=::NewEnumerator1, Loop Duration=1, UseLoopDelay=false)
   /Game/Effects/Niagara/Modules/Spawn/SpawnRateFromMultipleSpawnGroups_2(Global Spawnrate=50)
   &SpawnRate(SpawnRate=&Lerp_Float(A=10, B=25, Alpha=$User.ParaGlideLeanAlpha), Use Spawn Probability=false) {enabled=false}
  ]
  particleSpawnScript [
   &FN_SetSpawnGroup(Number of Spawn Groups=2)
   &InitializeParticle(
    AdjustAlpha=false
    AdjustHue=true
    AdjustSaturation=false
    AdjustValue=false
    Color=<rgba 2 0.5051 0 1>
    Color Mode=::NewEnumerator3
    Hue Shift Range:Vector2f=<-0.2 0.1>
    Lifetime=&FloatFromCurve(
     FloatCurve=DataInterface() {propertyValues {Curve {keys [
         {
          arriveTangent=0.1787
          arriveTangentWeight=0.9791
          interpMode=RCIM_Cubic
          leaveTangent=0.1787
          leaveTangentWeight=0.9791
          tangentMode=RCTM_User
          tangentWeightMode=RCTWM_WeightedBoth
          time=0
          value=0
         }
         {
          arriveTangent=8.2517
          arriveTangentWeight=0.8019
          interpMode=RCIM_Cubic
          leaveTangent=8.2517
          leaveTangentWeight=0.8019
          tangentMode=RCTM_User
          tangentWeightMode=RCTWM_WeightedBoth
          time=1
          value=1
         }
        ]}}}
     CurveIndex=&RandomRangeFloat(Minimum=0, Maximum=1)
    )
    Mesh Renderer Array Visibility Mode=::NewEnumerator3
    Mesh Renderer Info=DataInterface() {propertyValues {MeshRenderer=&self:ChunksHand_3.NiagaraMeshRendererProperties_0}}
    Mesh Scale Mode=::NewEnumerator1
    Mesh Uniform Scale Max=1
    Mesh Uniform Scale Min=0.35
    UsePositionOffset=false
    Write Material Random=true
   ) {module=InitializeParticle001}
   &SkeletalMeshLocation(
    Bone / Socket Index=$Particles.SpawnGroup
    Bone Sampling Mode=::NewEnumerator4
    Inherit Velocity=false
    Interpolate Mesh Sampling=false
    MaskSpawnGroup=false
    Skeletal Mesh=$User.SkeletalMesh
   )
   &ShapeLocation(
    Cylinder Height=20
    Cylinder Radius=10
    MaskSpawnGroup=false
    Override Seed=false
    Shape Primitive=::NewEnumerator1
    Surface Only=true
    Surface Only Band Thickness=1
    UseOwnerScale=true
   )
   &AddVelocity(
    Velocity=&FN_FrontEndVector3Modifier(
     Gameplay=&RandomRangeVector(Minimum=<50 50 220>, Maximum=<-50 -50 350>)
     Front End=&RandomRangeVector(Minimum=<0 0 85>, Maximum=<1 1 180>)
    )
    Velocity Speed Scale=0.5
   )
   &InitialMeshOrientation(Use Rotation Vector=false)
   &CameraOffset()
  ]
  particleUpdateScript [
   &ParticleState()
   &ScaleMeshSize(Scale Factor=&VectorFromFloat(Value=&FloatFromCurve(FloatCurve=DataInterface() {propertyValues {Curve {keys [
          {arriveTangent=9.0725, interpMode=RCIM_Cubic, leaveTangent=9.0725, tangentMode=RCTM_User, time=0, value=0}
          {arriveTangent=0.5485, interpMode=RCIM_Cubic, leaveTangent=0.5485, tangentMode=RCTM_User, time=0.1408, value=0.9847}
          {arriveTangent=-3.2166, interpMode=RCIM_Cubic, leaveTangent=-3.2166, tangentMode=RCTM_User, time=1, value=0}
         ]}}})))
   &Drag(Use Linear Drag=true, Use Rotational Drag=true)
   &SolveForcesAndVelocity(Clamp Velocity=false, Limit Acceleration=false)
  ]
  rendererClasses [&NiagaraMeshRendererProperties]
  renderers [{rendererClass=&NiagaraMeshRendererProperties, rendererIndex=0}]
 }
 ChunksHand001 {
  emitterUpdateScript [
   &EmitterState(Life Cycle Mode=::NewEnumerator1, Loop Duration=1, UseLoopDelay=false)
   &SpawnRateFromMultipleSpawnGroups_2(Global Spawnrate=50)
   &SpawnRate(SpawnRate=&Lerp_Float(A=10, B=25, Alpha=$User.ParaGlideLeanAlpha), Use Spawn Probability=false) {enabled=false}
  ]
  particleSpawnScript [
   &FN_SetSpawnGroup(Number of Spawn Groups=2)
   &InitializeParticle(
    AdjustAlpha=false
    AdjustHue=true
    AdjustSaturation=false
    AdjustValue=false
    Color=<rgba 2 0.5051 0 1>
    Color Mode=::NewEnumerator3
    Hue Shift Range:Vector2f=<-0.2 0.1>
    Lifetime Max=0.3
    Lifetime Min=0.1
    Lifetime Mode=::NewEnumerator1
    Mesh Renderer Array Visibility Mode=::NewEnumerator3
    Mesh Renderer Info=DataInterface() {propertyValues {MeshRenderer=&self:ChunksHand001_1.NiagaraMeshRendererProperties_0}}
    Mesh Scale Mode=::NewEnumerator1
    Mesh Uniform Scale Max=0.2
    Mesh Uniform Scale Min=0.35
    UsePositionOffset=false
    Write Material Random=true
   ) {module=InitializeParticle001}
   &SkeletalMeshLocation(
    Bone / Socket Index=$Particles.SpawnGroup
    Bone Sampling Mode=::NewEnumerator4
    Inherit Velocity=false
    Interpolate Mesh Sampling=false
    MaskSpawnGroup=false
    Skeletal Mesh=$User.SkeletalMesh
   )
   &ShapeLocation(
    Cylinder Height=150
    Cylinder Height Midpoint=0
    Cylinder Radius=10
    MaskSpawnGroup=false
    Override Seed=false
    Shape Primitive=::NewEnumerator1
    Surface Only=true
    Surface Only Band Thickness=1
    UseOwnerScale=true
   )
   &AddVelocity(
    Velocity=&FN_FrontEndVector3Modifier(
     Gameplay=&RandomRangeVector(Minimum=<50 50 220>, Maximum=<-50 -50 350>)
     Front End=&RandomRangeVector(Minimum=<0 0 85>, Maximum=<1 1 180>)
    )
    Velocity Speed Scale=0.5
   )
   &InitialMeshOrientation(Use Rotation Vector=false)
   &CameraOffset()
  ]
  particleUpdateScript [
   &ParticleState()
   &ScaleMeshSize(Scale Factor=&VectorFromFloat(Value=&FloatFromCurve(FloatCurve=DataInterface() {propertyValues {Curve {keys [
          {arriveTangent=9.0725, interpMode=RCIM_Cubic, leaveTangent=9.0725, tangentMode=RCTM_User, time=0, value=0}
          {arriveTangent=0.5485, interpMode=RCIM_Cubic, leaveTangent=0.5485, tangentMode=RCTM_User, time=0.1408, value=0.9847}
          {arriveTangent=-3.2166, interpMode=RCIM_Cubic, leaveTangent=-3.2166, tangentMode=RCTM_User, time=1, value=0}
         ]}}})))
   &Drag(Use Linear Drag=true, Use Rotational Drag=true)
   &SolveForcesAndVelocity(Clamp Velocity=false, Limit Acceleration=false)
  ]
  rendererClasses [&NiagaraMeshRendererProperties]
  renderers [{rendererClass=&NiagaraMeshRendererProperties, rendererIndex=0}]
 }
 Crystals001 {
  emitterSpawnScript [
   /SetVariables(Emitter.Shards Per limb=14) {isSetParameters=true}
   /SetVariables(Emitter.num limbs=2) {isSetParameters=true}
   /SetVariables(Emitter.total shards:NiagaraInt32=&Multiply_Int(A=$Emitter.Shards Per limb, B=$Emitter.num limbs)) {isSetParameters=true}
  ]
  emitterUpdateScript [
   &EmitterState()
   &SpawnBurst_Instantaneous(Spawn Count=$Emitter.total shards, Use Spawn Probability=false, UseLoopCountLimit=true)
  ]
  particleSpawnScript [
   &PartitionParticles(
    Normalized Index In Partition Includes One=true
    Number Of Partitions=$Emitter.num limbs
    Particles Per Partition:NiagaraInt32=$Emitter.Shards Per limb
    Partition Mode:ENiagara_PartitionMode=::NewEnumerator1
    Write Partition Index to Ribbon Link Order:ENiagara_AttributeSamplingApplyOutput=::NewEnumerator0
    Write Partition to Ribbon ID=::NewEnumerator0
   )
   &InitializeParticle(
    Color=$User.Crystal_Color
    Lifetime=5
    Mesh Renderer Array Visibility Mode=::NewEnumerator1
    Mesh Renderer Info=DataInterface() {propertyValues {MeshRenderer=&self:Crystals001_1.NiagaraMeshRendererProperties_2}}
    Mesh Renderer Mesh Index:NiagaraInt32=&MakeCustomIntFromBool(
     Bool=&SetBoolByIntComparison(A=$Particles.PartitionParticles.Partition, B=0, Comparison Type=::NewEnumerator4)
     True INT=&RandomRangeInt(Minimum=0, Maximum=13, Randomness Mode=::NewEnumerator2, Random Seed=?)
     False INT=&RandomRangeInt(Minimum=0, Maximum=13, Randomness Mode=::NewEnumerator2, Random Seed=?)
    )
    Mesh Scale=&Multiply_VectorByFloat(
     Vector=&RandomRangeVector(Minimum=<1 1 0.1>, Maximum=<0.5 0.5 0.5>, Random Seed=49, Fixed Random Seed=true)
     Float=&Lerp_Float(
      A=&Lerp_Float(
       A=6
       B=10
       Alpha=&FloatFromCurve(
        FloatCurve=DataInterface() {propertyValues {Curve {keys [<at 0 0 RCIM_Cubic>, {arriveTangent=2.8494, interpMode=RCIM_Cubic, leaveTangent=2.8494, tangentMode=RCTM_User, time=1, value=1}]}}}
        CurveIndex=&RandomRangeFloat(Minimum=0, Maximum=1, Random Seed=5204, Fixed Random Seed=true)
       )
      )
      B=&Lerp_Float(
       A=4
       B=3
       Alpha=&FloatFromCurve(
        FloatCurve=DataInterface() {propertyValues {Curve {keys [<at 0 0 RCIM_Cubic>, {arriveTangent=2.8494, interpMode=RCIM_Cubic, leaveTangent=2.8494, tangentMode=RCTM_User, time=1, value=1}]}}}
        CurveIndex=&RandomRangeFloat(Minimum=0, Maximum=1, Random Seed=241310, Fixed Random Seed=true)
       )
      )
      Alpha=&FloatFromCurve(
       FloatCurve=DataInterface() {propertyValues {Curve {keys [{tangentMode=RCTM_User, time=0, value=0}, {tangentMode=RCTM_User, time=1, value=1}]}}}
       CurveIndex=$Particles.PartitionParticles.Normalized Index In Partition
      )
     )
    )
    Mesh Scale Mode=::NewEnumerator4
    Ribbon Width=30
    Ribbon Width Mode=::NewEnumerator1
    Sprite Rotation Mode=::NewEnumerator0
    Sprite Size Mode=::NewEnumerator3
    Uniform Sprite Size=14.2073
    UsePositionOffset=false
    Write Material Random=true
   )
   &CameraOffset()
   &InitialMeshOrientation(Mesh Orientation Mode=::NewEnumerator7, Use Rotation Vector=false)
   /SetVariables(Particles.Scale 1:Vector3f=&Multiply_VectorByFloat(Vector=$Particles.Scale, Float=&RandomRangeFloat(Minimum=12, Maximum=15))) {enabled=false, isSetParameters=true}
   /SetVariables(Particles.bool:NiagaraBool=/Game/Effects/Niagara/DynamicInputs/State/FN_IsExecutionStateActive()) {isSetParameters=true}
   /SetVariables(Particles.Large index:NiagaraInt32=&RandomRangeInt(Minimum=0, Maximum=1, Randomness Mode=::NewEnumerator2, Random Seed=?)) {isSetParameters=true}
   /SetVariables(Particles.VisibilityTag=&MakeIntFromFloat(Float=&FloatFromCurve(
      FloatCurve=DataInterface() {propertyValues {Curve {keys [<at 0 1 RCIM_Constant>,<at 0.15 0>]}}}
      CurveIndex=$Particles.PartitionParticles.Normalized Index In Partition
     ))) {isSetParameters=true}
  ]
  particleUpdateScript [
   &ParticleState(Kill Particles When Lifetime Has Elapsed=false, Let Infinitely Lived Particles Die When Emitter Deactivates=true)
   &SampleSkeletalMesh(
    AdditionalYawPitchRoll=false
    Bone / Socket Index:NiagaraInt32=$Particles.PartitionParticles.Partition
    Bone Sampling Mode:ENiagaraBoneSamplingMode=::NewEnumerator4
    Sampled Position Offset:Vector3f=&Multiply_VectorByFloat(
     Vector=&MakeVector(
      X=&Waveform(
       "Waveform [1]"=::NewEnumerator12
       "[1] Input Position"=&RandomRangeFloat(Minimum=0, Maximum=100)
       "[1] Phase Scale"=&RandomRangeFloat(Minimum=1, Maximum=0.5)
       "[1] Frequency"=1
       "[1] Amplitude Min/Max"=&RandomRangeVector2D(Minimum=<-3 3>, Maximum=<-3 3>)
      )
      Y=&Waveform(
       "Waveform [1]"=::NewEnumerator12
       "[1] Input Position"=&RandomRangeFloat(Minimum=0, Maximum=100)
       "[1] Phase Scale"=&RandomRangeFloat(Minimum=1, Maximum=0.5)
       "[1] Frequency"=1
       "[1] Amplitude Min/Max"=&RandomRangeVector2D(Minimum=<-3 3>, Maximum=<-3 3>)
      )
      Z=&Waveform(
       "Waveform [1]"=::NewEnumerator12
       "[1] Input Position"=&RandomRangeFloat(Minimum=0, Maximum=100)
       "[1] Phase Scale"=&RandomRangeFloat(Minimum=0.5, Maximum=1)
       "[1] Frequency"=1
       "[1] Amplitude Min/Max"=&RandomRangeVector2D(Minimum=<-3 3>, Maximum=<-3 3>)
      )
     )
     Float=0.2
    )
    Skeletal Mesh:NiagaraDataInterfaceSkeletalMesh=$User.SkeletalMesh
   )
   &ShapeLocation(
    Additional Quaternion Rotation=true
    Cylinder Distribution=::NewEnumerator2
    Cylinder Height=&Lerp_Float(
     A=125
     B=180
     Alpha=&FloatFromCurve(
      FloatCurve=DataInterface() {propertyValues {Curve {keys [
          {interpMode=RCIM_Cubic, leaveTangentWeight=0.05, tangentMode=RCTM_User, time=0, value=0}
          {arriveTangentWeight=0.0901, interpMode=RCIM_Cubic, leaveTangentWeight=0.3906, tangentMode=RCTM_User, time=0.2, value=-0.1}
          {arriveTangentWeight=0.4024, interpMode=RCIM_Cubic, leaveTangentWeight=0.115, tangentMode=RCTM_User, time=0.8, value=1.1}
          {arriveTangentWeight=0.0667, interpMode=RCIM_Cubic, time=1, value=1}
         ]}}}
      CurveIndex=$System.Timeline.NormalizedTime
     )
    )
    Cylinder Height Midpoint=0
    Cylinder Radius=&Add_Float(
     A=&MakeCustomFloatFromBool(
      Bool=&SetBoolByIntComparison(A=$Particles.PartitionParticles.Partition, B=0, Comparison Type=::NewEnumerator4)
      True Float=&SelectFloatFromArray(
       Direct Array Index=$Particles.PartitionParticles.Index In Partition
       Float Selection Array=DataInterface() {propertyValues {FloatData [2,0,3,0]}}
      )
      False Float=&SelectFloatFromArray(
       Direct Array Index=$Particles.PartitionParticles.Index In Partition
       Float Selection Array=DataInterface() {propertyValues {FloatData [5,0,5,0]}}
      )
     )
     B=&Add_Float(
      A=&Lerp_Float(
       A=4
       B=15
       Alpha=&Lerp_Float(
        A=&FloatFromCurve(
         FloatCurve=DataInterface() {propertyValues {Curve {keys [
             {arriveTangent=0.7178, interpMode=RCIM_Cubic, leaveTangent=0.7178, tangentMode=RCTM_User, time=0, value=0.5}
             {interpMode=RCIM_Cubic, tangentMode=RCTM_User, time=0.5, value=1}
             {arriveTangent=-1.8799, tangentMode=RCTM_User, time=1, value=0.5}
            ]}}}
         CurveIndex=$Particles.PartitionParticles.Normalized Index In Partition
        )
        B=&FloatFromCurve(
         FloatCurve=DataInterface() {propertyValues {Curve {keys [
             {arriveTangent=3.9118, interpMode=RCIM_Cubic, leaveTangent=3.9118, tangentMode=RCTM_User, time=0, value=0}
             {arriveTangent=5.7587, interpMode=RCIM_Cubic, leaveTangent=5.7587, tangentMode=RCTM_User, time=1, value=4}
            ]}}}
         CurveIndex=$Particles.PartitionParticles.Normalized Index In Partition
        )
        Alpha=$System.Timeline.NormalizedTime
       )
      )
      B=&Multiply_Float(
       A=&FloatFromCurve(
        FloatCurve=DataInterface() {propertyValues {Curve {keys [
            <at 0 1 RCIM_Cubic>
            {
             arriveTangent=-5.0799
             arriveTangentWeight=1.0424
             interpMode=RCIM_Cubic
             leaveTangent=-5.0799
             leaveTangentWeight=1.0424
             tangentMode=RCTM_User
             tangentWeightMode=RCTWM_WeightedBoth
             time=1
             value=0
            }
           ]}}}
        CurveIndex=$Particles.PartitionParticles.Normalized Index In Partition
        Scale Curve=5
       )
       B=&RandomRangeFloat(Minimum=-1, Maximum=1)
      )
     )
    )
    MaskSpawnGroup=false
    Offset=<0 0 15>
    Offset Mode=::NewEnumerator0
    Rotation Mode=::NewEnumerator3
    Rotation Quaternion=&LerpQuaternion(
     Quaternion A:Quat4f=<0 0 0 1>
     Quaternion B:Quat4f=&FindQuatBetweenTwoVectors(
      From Vector=<0 0 -1>
      To Vector=&TransformVector(
       Vector=&Add_Vector(
        A=&Multiply_Vector(A=$Engine.Owner.Velocity, B=<1 1 1>)
        B=&FN_FrontEndVector3Modifier(Gameplay=<0 0 -999>, Front End=<0 0 -333>)
       )
       Source Space=::World
       Destination Space=::Simulation
      )
     )
     Lerp Factor:NiagaraFloat=&FloatFromCurve(
      FloatCurve=DataInterface() {propertyValues {Curve {keys [{tangentMode=RCTM_User, time=0, value=0}, {tangentMode=RCTM_User, time=1, value=1}]}}}
      CurveIndex=$Particles.PartitionParticles.Normalized Index In Partition
      Scale Curve=0.5
     )
     Enable Spherical Linear Interpolation:NiagaraBool=false
    )
    Shape Origin=$Particles.SampleSkeletalMesh.SampledPosition
    Shape Primitive=::NewEnumerator1
    U Position=&Add_Float(
     A=&FloatFromCurve(
      FloatCurve=DataInterface() {propertyValues {Curve {keys [{tangentMode=RCTM_User, time=0, value=0.05}, {tangentMode=RCTM_User, time=1, value=0.999}]}}}
      CurveIndex=$Particles.PartitionParticles.Normalized Index In Partition
      Scale Curve=-2
     )
     B=0
    )
    UseOwnerScale=true
    V Position=&FloatFromCurve(
     FloatCurve=DataInterface() {propertyValues {Curve {keys [{tangentMode=RCTM_User, time=0, value=0.05}, {tangentMode=RCTM_User, time=1, value=1}]}}}
     CurveIndex=$Particles.PartitionParticles.Normalized Index In Partition
    )
   )
   &UpdateMeshOrientation(
    Facing Direction=&Multiply_Vector(A=$Particles.ShapeLocation.ShapeVector, B=<1 1 1000>)
    Facing Interpolation=false
    Orientation Method=::NewEnumerator1
    Override Quat=false
    TransformFacing=false
    TransformUp=false
   ) {module=UpdateMeshOrientation002}
   &UpdateMeshOrientation(
    Rotation Rate=&Waveform(
     "Waveform [1]"=::NewEnumerator0
     "[1] Input Position"=&RandomRangeFloat(Minimum=0, Maximum=100)
     "[1] Phase Scale"=0.35
     "[1] Frequency"=1
     "[1] Amplitude Min/Max"=&Multiply_Vector2D_ByFloat(Float=0.01, Vector2D=<-1 1>)
    )
    Rotation Vector=<0 0 1>
   )
   &UpdateMeshOrientation(
    Rotation Rate=&Waveform(
     "Waveform [1]"=::NewEnumerator0
     "[1] Input Position"=&RandomRangeFloat(Minimum=0, Maximum=100)
     "[1] Phase Scale"=0.35
     "[1] Frequency"=1
     "[1] Amplitude Min/Max"=&Multiply_Vector2D_ByFloat(Float=0.01, Vector2D=<-1 1>)
    )
    Rotation Vector=<0 1 0>
   ) {module=UpdateMeshOrientation003}
   &UpdateMeshOrientation(Rotation Rate=&Waveform(
     "Waveform [1]"=::NewEnumerator0
     "[1] Input Position"=&RandomRangeFloat(Minimum=0, Maximum=100)
     "[1] Phase Scale"=0.35
     "[1] Frequency"=1
     "[1] Amplitude Min/Max"=&Multiply_Vector2D_ByFloat(Float=0.01, Vector2D=<-1 1>)
    )) {module=UpdateMeshOrientation001}
   &ScaleMeshSize(Scale Factor=&VectorFromFloat(Value=&Lerp_Float(
      A=0.6
      B=1
      Alpha=&FloatFromCurve(
       FloatCurve=DataInterface() {propertyValues {Curve {keys [
           {
            arriveTangentWeight=1.0423
            interpMode=RCIM_Cubic
            leaveTangentWeight=1.0423
            tangentMode=RCTM_User
            tangentWeightMode=RCTWM_WeightedBoth
            time=0
            value=1
           }
           {
            arriveTangent=-2.8575
            arriveTangentWeight=0.6279
            interpMode=RCIM_Cubic
            leaveTangent=-2.8575
            leaveTangentWeight=0.6279
            tangentMode=RCTM_User
            tangentWeightMode=RCTWM_WeightedBoth
            time=1
            value=0
           }
          ]}}}
       CurveIndex=$Particles.PartitionParticles.Normalized Index In Partition
      )
     )))
   /SetVariables(
    Particles.Scale2:Vector3f=&VectorFromFloat(Value=&RandomRangeFloat(Minimum=0.35, Maximum=0.4))
    Particles.Vector 4:Vector4f=&MakeVector4(
     X=&MakeFloatFromInt(Int=&ReturnParticleIDAttributes())
     Y=0
     Z=0
     W=0
    )
    Particles.float=10
   ) {isSetParameters=true}
  ]
  rendererClasses [&NiagaraMeshRendererProperties, &NiagaraSpriteRendererProperties]
  renderers [
   {rendererClass=&NiagaraMeshRendererProperties, rendererIndex=0}
   {rendererClass=&NiagaraMeshRendererProperties, rendererIndex=1}
   {rendererClass=&NiagaraSpriteRendererProperties, rendererIndex=2}
  ]
 }
 Decal {
  emitterUpdateScript [
   &EmitterState(Life Cycle Mode=::NewEnumerator1, Loop Duration=1, UseLoopDelay=false)
   &SpawnRateFromMultipleSpawnGroups_2(Global Spawnrate=4)
   &SpawnRate(SpawnRate=14, Use Spawn Probability=false) {enabled=false}
  ]
  particleSpawnScript [
   /Niagara/Modules/Spawn/Initialization/InitializeParticle(
    Color=<rgba 0.9062 0.4123 0.1067 0.6>
    Lifetime=0.6
    Module.Write Color=true
    Module.Write Lifetime=true
    Module.Write Mass=true
    Module.Write Material Random=true
    Module.Write Position=false
    Module.Write Scale=false
    Module.Write SpriteRotation=false
    Module.Write SpriteSize=true
    Position=&SimulationPosition()
    Sprite Size=<40 40>
   ) {enabled=false}
   /Niagara/Modules/Spawn/Initialization/V2/InitializeParticle(
    Color=&MultiplyLinearColor(
     Linear Color=&HueShiftLinearColor(Hue=0, Saturation=1, Value=1, Linear Color=$User.Crystal_Color)
     Scale Factor=&MakeFloat4FromVectorAndFloat("Vector (RGB)"=&VectorFromFloat(Value=2), "Float (Alpha)"=1)
    )
    Lifetime=0.6
    Mass Mode=::NewEnumerator1
    Sprite Size=<40 40>
    Sprite Size Mode=::NewEnumerator4
    UsePositionOffset=false
    Write Material Random=true
   ) {module=InitializeParticle001}
   &Color(Color=&HueShiftLinearColor(
     Hue=0
     Saturation=&MakeCustomFloatFromBool(Bool=$System.IsOrangeRed, True Float=1.25, False Float=2)
     Value=1.5
     Linear Color=$User.Crystal_Color
    )) {module=Color003}
   &CameraOffset(Camera Offset Amount=2)
   &PartitionParticles()
  ]
  particleUpdateScript [
   &ParticleState()
   &SkeletalMeshLocation(
    Bone / Socket Index=$Particles.PartitionParticles.Partition
    Bone Sampling Mode=::NewEnumerator4
    Inherit Velocity=false
    Interpolate Mesh Sampling=false
    MaskSpawnGroup=false
    Sampled Position Offset=&Lerp_Vector(A=<10 0 0>, B=<-10 1 1>, Alpha=&MakeFloatFromInt(Int=$Particles.PartitionParticles.Partition))
    Sampled Position Offset Coordinate Space:ENiagaraMeshTransforms=::NewEnumerator3
    Skeletal Mesh=$User.SkeletalMesh
   )
   &ScaleColor(
    Scale Alpha=&FloatFromCurve(FloatCurve=DataInterface() {propertyValues {Curve {keys [
         <at 0.0001 0 RCIM_Cubic>
         {arriveTangent=1.6347, interpMode=RCIM_Cubic, leaveTangent=1.6347, time=0.1393, value=0.6251}
         {arriveTangent=0.8334, interpMode=RCIM_Cubic, leaveTangent=0.8334, time=0.3202, value=0.5233}
         {arriveTangent=-0.4566, interpMode=RCIM_Cubic, leaveTangent=-0.4566, time=0.4613, value=0.8934}
         {arriveTangent=-0.922, interpMode=RCIM_Cubic, leaveTangent=-0.922, time=0.6241, value=0.3845}
         {arriveTangent=-1.0215, interpMode=RCIM_Cubic, leaveTangent=-1.0215, time=0.7724, value=0.6066}
         <at 1.0006 0 RCIM_Cubic>
        ]}}})
    ScaleA=true
    ScaleRGB=false
   )
   /SetVariables(Particles.Linear Color=&MultiplyLinearColor(
     Linear Color=$Particles.Color
     Scale Factor=&MakeFloat4FromVectorAndFloat("Vector (RGB)"=&VectorFromFloat(Value=5), "Float (Alpha)"=1)
    )) {isSetParameters=true}
  ]
  rendererClasses [&NiagaraSpriteRendererProperties]
  renderers [
   {rendererClass=&NiagaraSpriteRendererProperties}
   {rendererClass=&NiagaraSpriteRendererProperties, rendererIndex=1}
  ]
 }
 Empty {
  bEnabled=false
  emitterUpdateScript [
   &EmitterState(Life Cycle Mode=::NewEnumerator1, UseLoopDelay=false)
   &SpawnBurst_Instantaneous(Use Spawn Probability=false, UseLoopCountLimit=false)
  ]
  particleSpawnScript [&InitializeParticle(Color Mode=::NewEnumerator0, Lifetime=5, UsePositionOffset=false, Write Material Random=true)]
  particleUpdateScript [
   &ParticleState()
   &DynamicMaterialParameters(
    Index 0 Param 1=&MakeCustomFloatFromBool(Bool=$System.IsOrangeRed, True Float=1, False Float=0)
    Module.Index0Param0Write=true
    Module.Index0Param1Write=true
    Module.Index0Param2Write=true
    Module.Index0Param3Write=true
   )
  ]
  rendererClasses [&NiagaraSpriteRendererProperties]
  renderers [{rendererClass=&NiagaraSpriteRendererProperties}]
 }
 Minimal {
  emitterUpdateScript [
   &EmitterState()
   &SpawnBurst_Instantaneous(Spawn Count=6, Use Spawn Probability=false, UseLoopCountLimit=true)
  ]
  particleSpawnScript [
   &PartitionParticles(Normalized Index In Partition Includes One=true, Particles Per Partition=3, Partition Mode:ENiagara_PartitionMode=::NewEnumerator1)
   &InitializeParticle(Color=$User.Crystal_Color, Lifetime=5, UsePositionOffset=false, Write Material Random=true)
  ]
  particleUpdateScript [
   &ParticleState(Kill Particles When Lifetime Has Elapsed=false)
   &SampleSkeletalMesh(
    AdditionalYawPitchRoll=false
    Bone / Socket Index:NiagaraInt32=$Particles.PartitionParticles.Partition
    Bone Sampling Mode:ENiagaraBoneSamplingMode=::NewEnumerator4
    Sampled Position Offset:Vector3f=&MakeVector(
     X=&Waveform(
      "Waveform [1]"=::NewEnumerator12
      "[1] Input Position"=&RandomRangeFloat(Minimum=0, Maximum=1000)
      "[1] Phase Scale"=&RandomRangeFloat(Minimum=0.1, Maximum=0.035)
      "[1] Frequency"=1
      "[1] Amplitude Min/Max"=&RandomRangeVector2D(Minimum=<-3 3>, Maximum=<-3 3>)
     )
     Y=&Waveform(
      "Waveform [1]"=::NewEnumerator12
      "[1] Input Position"=&RandomRangeFloat(Minimum=0, Maximum=1000)
      "[1] Phase Scale"=&RandomRangeFloat(Minimum=0.1, Maximum=0.035)
      "[1] Frequency"=1
      "[1] Amplitude Min/Max"=&RandomRangeVector2D(Minimum=<-3 3>, Maximum=<-3 3>)
     )
     Z=&Waveform(
      "Waveform [1]"=::NewEnumerator12
      "[1] Input Position"=&RandomRangeFloat(Minimum=0, Maximum=1000)
      "[1] Phase Scale"=&RandomRangeFloat(Minimum=0.1, Maximum=0.035)
      "[1] Frequency"=1
      "[1] Amplitude Min/Max"=&RandomRangeVector2D(Minimum=<-3 3>, Maximum=<-3 3>)
     )
    )
    Skeletal Mesh:NiagaraDataInterfaceSkeletalMesh=$User.SkeletalMesh
   )
   &ShapeLocation(
    Additional Quaternion Rotation=true
    Cylinder Distribution=::NewEnumerator2
    Cylinder Height=&Lerp_Float(
     A=125
     B=180
     Alpha=&FloatFromCurve(
      FloatCurve=DataInterface() {propertyValues {Curve {keys [
          {interpMode=RCIM_Cubic, leaveTangentWeight=0.05, tangentMode=RCTM_User, time=0, value=0}
          {arriveTangentWeight=0.0901, interpMode=RCIM_Cubic, leaveTangentWeight=0.3906, tangentMode=RCTM_User, time=0.2, value=-0.1}
          {arriveTangentWeight=0.4024, interpMode=RCIM_Cubic, leaveTangentWeight=0.115, tangentMode=RCTM_User, time=0.8, value=1.1}
          {arriveTangentWeight=0.0667, interpMode=RCIM_Cubic, time=1, value=1}
         ]}}}
      CurveIndex=$System.Timeline.NormalizedTime
     )
    )
    Cylinder Height Midpoint=0
    Cylinder Radius=0
    MaskSpawnGroup=false
    Rotation Mode=::NewEnumerator3
    Rotation Quaternion=&LerpQuaternion(
     Quaternion A:Quat4f=<0 0 0 1>
     Quaternion B:Quat4f=&FindQuatBetweenTwoVectors(
      From Vector=<0 0 -1>
      To Vector=&TransformVector(
       Vector=&Add_Vector(
        A=&Multiply_Vector(A=$Engine.Owner.Velocity, B=<1 1 1>)
        B=&FN_FrontEndVector3Modifier(Gameplay=<0 0 -999>, Front End=<0 0 -333>)
       )
       Source Space=::World
       Destination Space=::Simulation
      )
     )
     Lerp Factor:NiagaraFloat=&FloatFromCurve(
      FloatCurve=DataInterface() {propertyValues {Curve {keys [{tangentMode=RCTM_User, time=0, value=0}, {tangentMode=RCTM_User, time=1, value=1}]}}}
      CurveIndex=$Particles.PartitionParticles.Normalized Index In Partition
      Scale Curve=0.5
     )
     Enable Spherical Linear Interpolation:NiagaraBool=false
    )
    Shape Origin=$Particles.SampleSkeletalMesh.SampledPosition
    Shape Primitive=::NewEnumerator1
    U Position=0.5
    UseOwnerScale=true
    V Position=&FloatFromCurve(
     FloatCurve=DataInterface() {propertyValues {Curve {keys [<at 0 0.75>,<at 1 0.25>]}}}
     CurveIndex=$Particles.PartitionParticles.Normalized Index In Partition
    )
   )
   &FN_Light_Attributes(
    Color=&HueShiftLinearColor(
     Hue=0
     Saturation=&MakeCustomFloatFromBool(Bool=$System.IsOrangeRed, True Float=1, False Float=1.5)
     Value=1
     Linear Color=$User.Crystal_Color
    )
    Create Color Attribute=true
    Create Exponent Attribute=true
    Create Radius Attribute=true
    Exponent=2
    Radius=30
    Scale Color=12
   )
  ]
  rendererClasses [&NiagaraLightRendererProperties]
  renderers [{rendererClass=&NiagaraLightRendererProperties, rendererIndex=0}]
 }
 Ribbons001 {
  emitterUpdateScript [
   &EmitterState(Life Cycle Mode=::NewEnumerator1, Loop Duration=1, UseLoopDelay=true)
   /Game/Effects/Niagara/Modules/Skydiving/FN_DivingAttributes/FN_EmitterDivingAttributes_v1(Bone Index=1, SkeletalMesh:NiagaraDataInterfaceSkeletalMesh=$User.SkeletalMesh)
   &FN_SpawnRibbon_v2(Number of Ribbons=2, Spawn Mode=::NewEnumerator0, SpawnRate=640)
  ]
  particleSpawnScript [
   &FN_SetSpawnGroup(Number of Spawn Groups=2)
   &FN_InitializeRibbon_v3(
    Color:LinearColor=$User.Crystal_Color
    Custom Offset=true
    Enable Skeletal Position=true
    Lifetime=0.35
    Local Vector Offset:Vector3f=<0 1 0>
    Ribbon Width=16.5
   )
   &Color(Color=&MultiplyLinearColor(
     Linear Color=&HueShiftLinearColor(Hue=0.05, Saturation=1, Value=2, Linear Color=$User.Crystal_Color)
     Scale Factor=&MakeFloat4FromVectorAndFloat("Vector (RGB)"=&VectorFromFloat(Value=1), "Float (Alpha)"=0.85)
    )) {module=Color003}
   &AddVelocity(
    Scale Added Velocity=&Lerp_Vector(A=<1 1 1>, B=<4 4 1>, Alpha=$System.Dive)
    Velocity=&FN_FrontEndVector3Modifier(Gameplay=<0 0 800>, Front End=<0 0 800>)
   )
   &SkeletalMeshLocation(
    Apply Sampled Orientation to Mesh Particles=true
    Bone / Socket Index=$Particles.SpawnGroup
    Bone Sampling Mode=::NewEnumerator4
    Inherit Velocity=false
    Interpolate Mesh Sampling=false
    MaskSpawnGroup=false
    Sampled Position Offset Coordinate Space:ENiagaraMeshTransforms=::NewEnumerator3
    Skeletal Mesh=$User.SkeletalMesh
   )
   &DynamicMaterialParameters(
    Index 0 Param 1=&MaskFloatBySpawnGroup(Masked Float=1, Passthrough Float=0.8)
    Index 0 Param 2=&MaskFloatBySpawnGroup(Masked Float=0.4, Passthrough Float=3)
    Index 0 Param 3=1
    Index 1 Param 1=&Lerp_Float(A=5, B=10, Alpha=$System.Dive)
    Module.Param0WriteEnabled=true
    Module.Param0WriteEnabled1=true
    Module.Param1WriteEnabled=true
    Module.Param1WriteEnabled1=false
    Module.Param2WriteEnabled=true
    Module.Param2WriteEnabled1=false
    Module.Param3WriteEnabled=false
    Module.Param3WriteEnabled1=false
    Write Parameter Index 1=true
   ) {enabled=false}
   &Color(Color=&HueShiftLinearColor(
     Hue=&FloatFromCurve(FloatCurve=DataInterface() {propertyValues {Curve {keys [
          <at 0 0.5 RCIM_Cubic>
          {
           arriveTangent=-3.9716
           arriveTangentWeight=0.435
           interpMode=RCIM_Cubic
           leaveTangent=-3.9716
           leaveTangentWeight=1.6184
           time=0.0267
           value=0
          }
          {arriveTangentWeight=0.0331, interpMode=RCIM_Cubic, leaveTangentWeight=0.0716, time=0.1259, value=0}
          <at 0.3408 0 RCIM_Cubic>
          {arriveTangentWeight=0.197, interpMode=RCIM_Cubic, leaveTangentWeight=0.3213, time=0.5005, value=0}
          {arriveTangentWeight=0.0868, interpMode=RCIM_Cubic, leaveTangentWeight=0.0797, time=0.761, value=0}
          <at 1 0 RCIM_Cubic>
         ]}}})
     Saturation=1
     Value=1
     Linear Color=$User.Crystal_Color
    )) {enabled=false, module=Color002}
  ]
  particleUpdateScript [
   &InheritSourceMovement(Applied Position Falloff Distance=100, Applied Position Scale=<0.9 0.9 1>, ApplyForce=false, ApplyPosition=true, ApplyPositionFalloff=false)
   &SolveForcesAndVelocity(Clamp Velocity=false, Limit Acceleration=false)
   &FN_UpdateRibbon(Detach Locked Particle=true, Enable Skeletal Position=true)
   &Color(Color=&HueShiftLinearColor(
     Hue=&SelectFloatFromArray(
      Direct Array Index=$System.COlor to int
      Direct Array Mode=::NewEnumerator1
      Float Selection Array=DataInterface() {propertyValues {FloatData [0.1,0.08,0.6,0.5,0.5,0.5,0.5,0.5,0.5,0.5,0.5,0.5]}}
     )
     Saturation=1
     Value=1
     Linear Color=&MultiplyLinearColor(
      Linear Color=&ColorFromCurve(ColorCurve=DataInterface() {propertyValues {
         AlphaCurve {keys [<at 0 1 RCIM_Cubic>,<at 1 1 RCIM_Cubic>]}
         BlueCurve {keys [
           {arriveTangent=-48.9212, interpMode=RCIM_Cubic, leaveTangent=-48.9212, tangentMode=RCTM_User, time=0.0267, value=12.3946}
           {arriveTangentWeight=2.8439, interpMode=RCIM_Cubic, leaveTangentWeight=4.8244, tangentMode=RCTM_User, time=0.1259, value=1}
           <at 0.3408 2>
           {arriveTangent=-3.5634, interpMode=RCIM_Cubic, leaveTangent=-3.5634, time=0.5005, value=2}
           <at 0.761 0.5029 RCIM_Cubic>
          ]}
         GreenCurve {keys [
           {arriveTangent=-102.7431, interpMode=RCIM_Cubic, leaveTangent=-102.7431, tangentMode=RCTM_User, time=0.0267, value=8.5704}
           {
            arriveTangent=-3.0599
            arriveTangentWeight=2.5623
            interpMode=RCIM_Cubic
            leaveTangent=-3.0599
            leaveTangentWeight=4.3468
            tangentMode=RCTM_User
            time=0.1259
            value=0.847
           }
           <at 0.3408 0.454>
           {arriveTangentWeight=0.1167, interpMode=RCIM_Cubic, leaveTangentWeight=0.102, tangentMode=RCTM_User, time=0.5005, value=0.454}
           <at 0.761 0.5029 RCIM_Cubic>
          ]}
         RedCurve {keys [
           <at 0.0267 0 RCIM_Cubic>
           {arriveTangentWeight=0.0854, interpMode=RCIM_Cubic, leaveTangentWeight=0.145, tangentMode=RCTM_User, time=0.1259, value=0}
           <at 0.3408 0.5032>
           {arriveTangent=-1.1978, interpMode=RCIM_Cubic, leaveTangent=-1.1978, time=0.5005, value=0.5032}
           <at 0.761 0 RCIM_Cubic>
          ]}
        }})
      Scale Factor=&MakeFloat4FromVectorAndFloat("Vector (RGB)"=&VectorFromFloat(Value=2.1588), "Float (Alpha)"=1)
     )
    )) {enabled=false}
   &ScaleRibbonWidth(Ribbon Width Scale=&FloatFromCurve(FloatCurve=DataInterface() {propertyValues {Curve {keys [<at 0 1 RCIM_Cubic>, {arriveTangent=-2.5144, interpMode=RCIM_Cubic, leaveTangent=-2.5144, tangentMode=RCTM_User, time=1, value=0.2}]}}}))
   /SetVariables(Particles.Vector 4:Vector4f=&MakeVector4(
     X=&MakeFloatFromInt(Int=&self:Color_teest(Linear Color:LinearColor=$User.Crystal_Color))
     Y=0
     Z=0
     W=0
    )) {enabled=false, isSetParameters=true}
   &Color(Color=&MultiplyLinearColor(
     Linear Color=&HueShiftLinearColor(
      Hue=&MakeCustomFloatFromBool(
       Bool=$System.IsOrangeRed
       True Float=&FloatFromCurve(FloatCurve=DataInterface() {propertyValues {Curve {keys [
            {arriveTangentWeight=0.435, interpMode=RCIM_Cubic, leaveTangentWeight=1.6184, time=-0.0331, value=0.0199}
            {arriveTangentWeight=0.0373, interpMode=RCIM_Cubic, leaveTangentWeight=0.0745, tangentMode=RCTM_User, time=0.0669, value=0.0199}
            {arriveTangentWeight=0.0712, interpMode=RCIM_Cubic, leaveTangentWeight=0.0712, tangentMode=RCTM_User, time=0.2669, value=-0.0301}
            {arriveTangentWeight=0.0696, interpMode=RCIM_Cubic, leaveTangentWeight=0.1044, tangentMode=RCTM_User, time=0.3, value=-0.0301}
            {arriveTangentWeight=0.1221, interpMode=RCIM_Cubic, leaveTangentWeight=0.0814, tangentMode=RCTM_User, time=0.7669, value=0.2199}
           ]}}})
       False Float=&FloatFromCurve(FloatCurve=DataInterface() {propertyValues {Curve {keys [
            {arriveTangentWeight=0.435, interpMode=RCIM_Cubic, leaveTangentWeight=1.6184, time=-0.0331, value=0.05}
            {arriveTangentWeight=0.0373, interpMode=RCIM_Cubic, leaveTangentWeight=0.0745, tangentMode=RCTM_User, time=0.0669, value=0.05}
            {arriveTangentWeight=0.0712, interpMode=RCIM_Cubic, leaveTangentWeight=0.0712, tangentMode=RCTM_User, time=0.2669, value=0}
            {arriveTangentWeight=0.0696, interpMode=RCIM_Cubic, leaveTangentWeight=0.1044, tangentMode=RCTM_User, time=0.3, value=0}
            {arriveTangentWeight=0.1221, interpMode=RCIM_Cubic, leaveTangentWeight=0.0814, tangentMode=RCTM_User, time=0.7669, value=0.25}
           ]}}})
      )
      Saturation=&MakeCustomFloatFromBool(Bool=$System.IsOrangeRed, True Float=1.25, False Float=1.8)
      Value=&FloatFromCurve(FloatCurve=DataInterface() {propertyValues {Curve {keys [
           {arriveTangent=-75.9372, interpMode=RCIM_Cubic, leaveTangent=-75.9372, tangentMode=RCTM_User, time=0, value=6}
           {
            arriveTangent=-3.2097
            arriveTangentWeight=0.67
            interpMode=RCIM_Cubic
            leaveTangent=-3.2097
            leaveTangentWeight=2.68
            tangentMode=RCTM_User
            time=0.2
            value=2
           }
           <at 1 1>
          ]}}})
      Linear Color=$User.Crystal_Color
     )
     Scale Factor=&MakeFloat4FromVectorAndFloat("Vector (RGB)"=&VectorFromFloat(Value=1), "Float (Alpha)"=0.85)
    )) {module=Color001}
   &Color(Color=$User.Crystal_Color) {enabled=false, module=Color004}
  ]
  rendererClasses [&NiagaraRibbonRendererProperties, &NiagaraSpriteRendererProperties]
  renderers [
   {rendererClass=&NiagaraRibbonRendererProperties}
   {rendererClass=&NiagaraSpriteRendererProperties, rendererIndex=1}
  ]
 }
 Sparks {
  emitterUpdateScript [
   &EmitterState(Life Cycle Mode=::NewEnumerator1, Loop Duration=0, UseLoopDelay=false)
   &SpawnRateFromMultipleSpawnGroups_2(Global Spawnrate=20)
   &SpawnRate(SpawnRate=&Lerp_Float(A=10, B=25, Alpha=$User.ParaGlideLeanAlpha), Use Spawn Probability=false) {enabled=false}
  ]
  particleSpawnScript [
   &FN_SetSpawnGroup(Number of Spawn Groups=2)
   &InitializeParticle(
    Color Maximum=&MultiplyLinearColor(
     Linear Color=$User.Crystal_Color
     Scale Factor=&MakeFloat4FromVectorAndFloat("Vector (RGB)"=&VectorFromFloat(Value=10), "Float (Alpha)"=1)
    )
    Color Minimum=&MultiplyLinearColor(
     Linear Color=$User.Crystal_Color
     Scale Factor=&MakeFloat4FromVectorAndFloat("Vector (RGB)"=&VectorFromFloat(Value=2), "Float (Alpha)"=1)
    )
    Color Mode=::NewEnumerator2
    Lifetime Max=0.5
    Lifetime Min=0.15
    Lifetime Mode=::NewEnumerator1
    Sprite Size Max=<2 5>
    Sprite Size Min=<3 2>
    Sprite Size Mode=::NewEnumerator2
    UsePositionOffset=false
    Write Material Random=true
   ) {module=InitializeParticle001}
   &ShapeLocation(MaskSpawnGroup=false, Override Seed=false, Sphere Radius=5, UseOwnerScale=true)
   &AddVelocity(
    Velocity=&FN_FrontEndVector3Modifier(
     Gameplay=&RandomRangeVector(Minimum=<100 100 0>, Maximum=<-100 -100 0>)
     Front End=&RandomRangeVector(Minimum=<50 50 0>, Maximum=<-50 -50 0>)
    )
    Velocity Speed Scale=0.5
   ) {enabled=false}
  ]
  particleUpdateScript [
   &ParticleState()
   &InheritVelocity(Inherited Velocity Amount Scale=<-0.03 -0.03 -0.03>, Inherited Velocity Speed Limit=1000, Speed Limit=true, UseSpeedThreshold=false) {enabled=false}
   &SkeletalMeshLocation(
    Bone / Socket Index=$Particles.SpawnGroup
    Bone Sampling Mode=::NewEnumerator4
    Inherit Velocity=false
    Interpolate Mesh Sampling=false
    MaskSpawnGroup=false
    Sampled Position Offset=&Add_Vector(
     A=&Lerp_Vector(
      A=<0 0 0>
      B=<0 0 100>
      Alpha=&FloatFromCurve(FloatCurve=DataInterface() {propertyValues {Curve {keys [{tangentMode=RCTM_User, time=0, value=0}, {tangentMode=RCTM_User, time=1, value=1}]}}})
     )
     B=&Lerp_Vector(
      A=<0 0 0>
      B=&RandomRangeVector(Minimum=<-10 -10 0>, Maximum=<10 10 1>)
      Alpha=&FloatFromCurve(FloatCurve=DataInterface() {propertyValues {Curve {keys [{arriveTangent=2.4314, interpMode=RCIM_Cubic, time=0, value=0.98}, <at 1 1 RCIM_Cubic>]}}})
     )
    )
    Skeletal Mesh=$User.SkeletalMesh
   )
   &Drag(Drag=2, Use Linear Drag=true, Use Rotational Drag=true) {enabled=false}
   &GravityForce(Gravity=<0 0 980>) {enabled=false}
   &SolveForcesAndVelocity(Clamp Velocity=false, Limit Acceleration=false)
   &ScaleSpriteSizeBySpeed(Max Scale Factor=<1 10>, Min Scale Factor=<0.5 4>)
   &ScaleColor(
    Scale Alpha=&FloatFromCurve(FloatCurve=DataInterface() {propertyValues {Curve {keys [<at 0 1>,<at 1 0>]}}})
    Scale RGB=&VectorFromFloat(Value=3)
    ScaleA=true
    ScaleRGB=true
   )
   /SetVariables(Particles.displayVel:Vector3f=&self:ScratchDynamicInput_03()) {enabled=false, isSetParameters=true}
   /SetVariables(Particles.displayVel:Vector3f=&Lerp_Vector(A=<0 0 1>, B=&RandomVector(Vector Scale=1), Alpha=0.1)) {isSetParameters=true}
  ]
  rendererClasses [&NiagaraSpriteRendererProperties]
  renderers [{rendererClass=&NiagaraSpriteRendererProperties}]
 }
}
summary {
 systemName=NS_Contrail_SknowKnight
 emitters [
  {emitterName=Ribbons001, rendererClasses [&NiagaraRibbonRendererProperties, &NiagaraSpriteRendererProperties]}
  {emitterName=ChunksHand, rendererClasses [&NiagaraMeshRendererProperties]}
  {emitterName=Sparks, rendererClasses [&NiagaraSpriteRendererProperties]}
  {emitterName=Decal, rendererClasses [&NiagaraSpriteRendererProperties]}
  {emitterName=ChunksHand001, rendererClasses [&NiagaraMeshRendererProperties]}
  {emitterName=Crystals001, rendererClasses [&NiagaraMeshRendererProperties, &NiagaraSpriteRendererProperties]}
  {bEnabled=false, emitterName=Empty, rendererClasses [&NiagaraSpriteRendererProperties]}
  {emitterName=Minimal, rendererClasses [&NiagaraLightRendererProperties]}
 ]
 userVariables [
  User.bIsFrontEnd {type=NiagaraBool}
  User.bIsFrontEndPreview {type=NiagaraBool}
  User.Crystal_Color {type=LinearColor, description="", defaultValue {struct=LinearColor, value=<rgba 0.4948 0.0742 0 1>}}
  User.ParaGlideLeanAlpha {type=NiagaraFloat, defaultValue=0, description=""}
  User.RibbonIDOffset {type=NiagaraInt32}
  User.SkeletalMesh {
   type=NiagaraDataInterfaceSkeletalMesh
   defaultValue {
    struct=NiagaraExt_VariableValue_DataInterface
    value {dataInterface=&self:NiagaraDataInterfaceSkeletalMesh_0}
   }
  }
 ]
}
